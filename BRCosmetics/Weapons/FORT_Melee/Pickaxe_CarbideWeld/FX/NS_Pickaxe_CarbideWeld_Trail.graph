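# Niagara swing trail for the CarbideWeld pickaxe, with color sets selected per cosmetic variant.
# Trails_Temp builds an anim trail from a shared System socket reader and draws three ribbon renderers, with light and dark background colors scaled by a time-of-day multiplier and TrailFade.
# Hexagons spawns short-lived SubUV sprites per distance moved, sampled at interpolated socket locations, with inherited velocity, drag and smaller sizing in the front end.
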NiagaraSystem /BRCosmetics/Weapons/FORT_Melee/Pickaxe_CarbideWeld/FX/NS_Pickaxe_CarbideWeld_Trail
dependencies {
 usedDataInterfaces [
  &NiagaraDataInterfaceCurve
  &NiagaraDataInterfaceColorCurve
  &NiagaraDataInterfaceVector2DCurve
  &NiagaraDataInterfaceSpriteRendererInfo
  &NiagaraDataInterfaceSocketReader
 ]
 usedDynamicInputs [
  &Vector2DFromCurve
  &self:SectionIndexFloatSelector_03
 ]
 usedModules [
  &self:SystemSpawnScript.SystemScriptSource.SystemScriptGraph.NiagaraNodeAssignment_6.SetVariables_#1
  &self:SystemSpawnScript.SystemScriptSource.SystemScriptGraph.NiagaraNodeAssignment_4.SetVariables_#2
  &FN_VariantManager
  &SystemState
  &self:SystemSpawnScript.SystemScriptSource.SystemScriptGraph.NiagaraNodeAssignment_0.SetVariables_#3
  &self:SystemSpawnScript.SystemScriptSource.SystemScriptGraph.NiagaraNodeAssignment_5.SetVariables_#4
 ]
}
emitters {
 Hexagons {
  emitterUpdateScript [
   &EmitterState()
   &SpawnPerUnit(Module.Use Max Movement Threshold=true, Module.Use Movement Tolerance=false, Spawn Spacing=15, Use Spawn Probability=false)
  ]
  particleSpawnScript [
   &InitializeParticle(
    Lifetime Max=0.4
    Lifetime Min=0.2
    Lifetime Mode=::NewEnumerator1
    Sprite Size Mode=::NewEnumerator3
    Uniform Sprite Size=&Multiply_Float(
     A=&RandomRangeFloat(Minimum=4, Maximum=6)
     B=&FN_FrontEndFloatModifier(Gameplay=1, Front End=0.75)
    )
    UsePositionOffset=false
    Write Material Random=true
   ) {module=InitializeParticle001}
   &CameraOffset(Camera Offset Amount=40)
   &DynamicMaterialParameters(
    Index 0 Param 2=4
    Module.Index0Float0Write=true
    Module.Index0Float1Write=true
    Module.Index0Float2Write=true
    Module.Index0Float3Write=true
   ) {module=DynamicMaterialParameters001}
   &SocketLocation(
    Interpolate Socket Sampling=true
    SamplePosOffset=true
    Sampled Position Offset=&RandomRangeVector(Minimum=<0 0 25>, Maximum=<0 0 75>)
    Sampled Position Offset Coordinate Space:ENiagaraSkelMeshTransforms=::NewEnumerator2
    Socket Reader=$System.Socket Reader
    Socket Sampling Mode=::NewEnumerator2
    VelocitySpeedLimit=false
   )
  ]
  particleUpdateScript [
   &InheritVelocity(
    Inherited Velocity Amount Scale=&Multiply_VectorByFloat(
     Vector=<0.05 0.05 0.05>
     Float=&FloatFromCurve(FloatCurve=DataInterface() {propertyValues {Curve {
         defaultValue=1
         keys [{interpMode=RCIM_Cubic, tangentMode=RCTM_User, time=0, value=1}, {interpMode=RCIM_Cubic, tangentMode=RCTM_User, time=1, value=0}]
        }}})
    )
    Speed Limit=true
    UseSpeedThreshold=false
   )
   &Drag(Drag=4, Use Linear Drag=true, Use Rotational Drag=true)
   &SolveForcesAndVelocity(Clamp Velocity=false, Limit Acceleration=false)
   &ParticleState()
   &ScaleSpriteSize(Uniform Curve Sprite Scale=DataInterface() {propertyValues {Curve {keys [
        {arriveTangent=5.201, interpMode=RCIM_Cubic, leaveTangent=5.201, tangentMode=RCTM_User, time=0, value=1}
        {arriveTangent=-0.2045, interpMode=RCIM_Cubic, leaveTangent=-0.2045, tangentMode=RCTM_User, time=0.2, value=1.7}
        {arriveTangent=-1.983, interpMode=RCIM_Cubic, leaveTangent=-1.983, tangentMode=RCTM_User, time=1, value=0}
       ]}}})
   &ScaleColor(
    Scale Alpha=&Multiply_Float(
     A=&FloatFromCurve(FloatCurve=DataInterface() {propertyValues {Curve {keys [<at 0.75 1>,<at 1 0>]}}})
     B=0.25
    )
    Scale RGB=&MakeVectorFromLinearColorRGB(Color=&VariantColorSelector(
      Variant Amount=2
      Original Color=<rgba 1 0.2168 0 1>
      Variant 1 Color=<rgba 0.4531 0.0203 0 1>
      Variant 2 Color=<rgba 0.2484 0.45 0 1>
      Variant 3 Color=<rgba 0.6768 0.1429 1 1>
     ))
    ScaleA=true
    ScaleRGB=true
   ) {module=ScaleColor001}
   &SubUVAnimation(
    End Frame Range Override=7
    "Sprite Renderer "=DataInterface() {propertyValues {SpriteRenderer=&self:Hexagons_0.NiagaraSpriteRendererProperties_1}}
    UseEndFrame=true
    UseStartFrame=true
   )
  ]
  rendererClasses [&NiagaraSpriteRendererProperties]
  renderers [{rendererClass=&NiagaraSpriteRendererProperties}]
 }
 Trails_Temp {
  emitterUpdateScript [
   &EmitterState()
   &FN_SpawnAnimTrail(Spawn Interval=0)
  ]
  particleSpawnScript [
   &FN_InitializeAnimTrailBoneIndecies()
   /Game/Effects/Niagara/Modules/AnimTrail/FN_InitializeAnimTrail/FN_InitializeAnimTrail_SocketReader/FN_InitializeAnimTrail_SocketReader(Lifetime=0.22, Socket Reader:NiagaraDataInterfaceSocketReader=$System.Socket Reader)
  ]
  particleUpdateScript [
   /Game/Effects/Niagara/Modules/AnimTrail/FN_UpdateAnimTrail/FN_UpdateAnimTrail_SocketReader/FN_UpdateAnimTrail_SocketReader(Socket Reader:NiagaraDataInterfaceSocketReader=$System.Socket Reader)
   &DynamicMaterialParameters(
    Index 0 Param 1=1
    Index 0 Param 2=1
    Module.Index0Float0Write=true
    Module.Index0Float1Write=true
    Module.Index0Float2Write=true
    Module.Index0Float3Write=true
   )
   &ScaleColor(
    Scale Alpha=&Multiply_Float(
     A=&MakeFloatFromLinearColor(LinearColor=$System.TODColorMultiplier, Channel:ENiagaraLinearColor_Channels=::NewEnumerator3)
     B=&Multiply_Float(
      A=&FloatFromCurve(FloatCurve=DataInterface() {propertyValues {Curve {keys [{interpMode=RCIM_Cubic, tangentMode=RCTM_User, time=0, value=1}, {interpMode=RCIM_Cubic, tangentMode=RCTM_User, time=1, value=0}]}}})
      B=$System.TrailFade
     )
    )
    Scale RGB=&MakeVectorFromLinearColorRGB(Color=&VariantColorSelector(
      Variant Amount=2
      Original Color=<rgba 1 0.5158 0 1>
      Variant 1 Color=<rgba 2 0.3701 0.2046 1>
      Variant 2 Color=<rgba 0.8208 1 0.6 1>
      Variant 3 Color=<rgba 0.8755 0.3542 1 1>
     ))
    ScaleA=true
    ScaleRGB=true
   )
   /SetVariables(Particles.Light_BG_Color:LinearColor=&MakeLinearColorFromVectorAndFloat(
     "Vector (RGB)"=&MakeVectorFromLinearColorRGB(Color=&MultiplyLinearColor(
       Linear Color=&VariantColorSelector(
        Variant Amount=2
        Original Color=&ColorFromCurve(ColorCurve=DataInterface() {propertyValues {
           AlphaCurve {keys [<at 0 1 RCIM_Cubic>,<at 1 0 RCIM_Cubic>]}
           BlueCurve {keys [<at 0 0>,<at 0.2525 0 RCIM_Cubic>,<at 1 0 RCIM_Cubic>]}
           GreenCurve {keys [<at 0 1.8619>, {arriveTangent=-1.7589, interpMode=RCIM_Cubic, leaveTangent=-1.7589, time=0.2525, value=0.4252}, <at 1 0.103 RCIM_Cubic>]}
           RedCurve {keys [<at 0 5>, {arriveTangent=-4, interpMode=RCIM_Cubic, leaveTangent=-4, time=0.2525, value=1}, <at 1 1 RCIM_Cubic>]}
          }})
        Variant 1 Color=&ColorFromCurve(ColorCurve=DataInterface() {propertyValues {
           AlphaCurve {keys [<at 0 1 RCIM_Cubic>,<at 1 0 RCIM_Cubic>]}
           BlueCurve {keys [<at 0 0 RCIM_Cubic>,<at 0.2632 0.0026>,<at 1 0 RCIM_Cubic>]}
           GreenCurve {keys [<at 0 1.206 RCIM_Cubic>,<at 0.2632 0.2068>,<at 1 0.0333 RCIM_Cubic>]}
           RedCurve {keys [<at 0 5 RCIM_Cubic>,<at 0.2632 0.8938>,<at 1 0.2552 RCIM_Cubic>]}
          }})
        Variant 2 Color=&ColorFromCurve(ColorCurve=DataInterface() {propertyValues {
           AlphaCurve {keys [<at 0 1 RCIM_Cubic>,<at 1 0 RCIM_Cubic>]}
           BlueCurve {keys [<at 0 0.0638>,<at 0.2646 0.0777>,<at 1 0.0246 RCIM_Cubic>]}
           GreenCurve {keys [<at 0 1>,<at 0.2646 1>,<at 1 0.1562 RCIM_Cubic>]}
           RedCurve {keys [<at 0 0.3198>,<at 0.2646 0.395>,<at 1 0.0214 RCIM_Cubic>]}
          }})
        Variant 3 Color=&ColorFromCurve(ColorCurve=DataInterface() {propertyValues {
           AlphaCurve {keys [<at 0 1 RCIM_Cubic>,<at 1 0 RCIM_Cubic>]}
           BlueCurve {keys [<at 0 5>, {arriveTangent=-4.724, interpMode=RCIM_Cubic, leaveTangent=-4.724, time=0.2695, value=1}, <at 1 0.276 RCIM_Cubic>]}
           GreenCurve {keys [<at 0 0.2604>, {arriveTangent=-0.2307, interpMode=RCIM_Cubic, leaveTangent=-0.2307, time=0.2695, value=0.157}, <at 1 0.0297 RCIM_Cubic>]}
           RedCurve {keys [<at 0 2.7937>, {arriveTangent=-2.6829, interpMode=RCIM_Cubic, leaveTangent=-2.6829, time=0.2695, value=0.9684}, <at 1 0.1108 RCIM_Cubic>]}
          }})
       )
       Scale Factor=&MakeVector4_FromLinearColor(LinearColor=&LerpLinearColors(StartColor=$Particles.Color, EndColor=$System.TODColorMultiplier, LerpFactor=0.4))
      ))
     "Float (Alpha)"=&Multiply_Float(
      A=&MakeFloatFromLinearColor(LinearColor=$System.TODColorMultiplier, Channel:ENiagaraLinearColor_Channels=::NewEnumerator3)
      B=&Multiply_Float(
       A=&FloatFromCurve(FloatCurve=DataInterface() {propertyValues {Curve {keys [{interpMode=RCIM_Cubic, tangentMode=RCTM_User, time=0, value=1}, {interpMode=RCIM_Cubic, tangentMode=RCTM_User, time=1, value=0}]}}})
       B=$System.TrailFade
      )
     )
    )) {isSetParameters=true}
   /SetVariables(Particles.Dark_BG_Color:LinearColor=&MakeLinearColorFromVectorAndFloat(
     "Vector (RGB)"=&MakeVectorFromLinearColorRGB(Color=&MultiplyLinearColor(
       Linear Color=&VariantColorSelector(
        Variant Amount=2
        Original Color=&ColorFromCurve(ColorCurve=DataInterface() {propertyValues {
           AlphaCurve {keys [<at 0 1 RCIM_Cubic>,<at 1 0 RCIM_Cubic>]}
           BlueCurve {keys [<at 0 0 RCIM_Cubic>,<at 1 0 RCIM_Cubic>]}
           GreenCurve {keys [<at 0 0.1034 RCIM_Cubic>,<at 1 0.0152 RCIM_Cubic>]}
           RedCurve {keys [<at 0 0.2865 RCIM_Cubic>,<at 1 0.1354 RCIM_Cubic>]}
          }})
        Variant 1 Color=&ColorFromCurve(ColorCurve=DataInterface() {propertyValues {
           AlphaCurve {keys [<at 0 1 RCIM_Cubic>,<at 1 0 RCIM_Cubic>]}
           BlueCurve {keys [<at 0 0 RCIM_Cubic>,<at 1 0 RCIM_Cubic>]}
           GreenCurve {keys [<at 0 0.0794 RCIM_Cubic>,<at 1 0.0037 RCIM_Cubic>]}
           RedCurve {keys [<at 0 0.2865 RCIM_Cubic>,<at 1 0.1354 RCIM_Cubic>]}
          }})
        Variant 2 Color=&ColorFromCurve(ColorCurve=DataInterface() {propertyValues {
           AlphaCurve {keys [<at 0 1 RCIM_Cubic>,<at 1 0 RCIM_Cubic>]}
           BlueCurve {keys [<at 0 0.24 RCIM_Cubic>,<at 1 0.0441 RCIM_Cubic>]}
           GreenCurve {keys [<at 0 0.6 RCIM_Cubic>,<at 1 0.1354 RCIM_Cubic>]}
           RedCurve {keys [<at 0 0.5144 RCIM_Cubic>,<at 1 0.1215 RCIM_Cubic>]}
          }})
        Variant 3 Color=&ColorFromCurve(ColorCurve=DataInterface() {propertyValues {
           AlphaCurve {keys [<at 0 1 RCIM_Cubic>,<at 1 0 RCIM_Cubic>]}
           BlueCurve {keys [<at 0 0.2865 RCIM_Cubic>,<at 1 0.1354 RCIM_Cubic>]}
           GreenCurve {keys [<at 0 0.1447 RCIM_Cubic>,<at 1 0.0315 RCIM_Cubic>]}
           RedCurve {keys [<at 0 0.2626 RCIM_Cubic>,<at 1 0.0988 RCIM_Cubic>]}
          }})
       )
       Scale Factor=&MakeVector4_FromLinearColor(LinearColor=$System.TODColorMultiplier)
      ))
     "Float (Alpha)"=&Multiply_Float(
      A=&MakeFloatFromLinearColor(LinearColor=$System.TODColorMultiplier, Channel:ENiagaraLinearColor_Channels=::NewEnumerator3)
      B=&Multiply_Float(
       A=&FloatFromCurve(
        FloatCurve=DataInterface() {propertyValues {Curve {keys [{interpMode=RCIM_Cubic, tangentMode=RCTM_User, time=0, value=1}, {interpMode=RCIM_Cubic, tangentMode=RCTM_User, time=1, value=0}]}}}
        Scale Curve=2
       )
       B=$System.TrailFade
      )
     )
    )) {isSetParameters=true}
  ]
  rendererClasses [&NiagaraRibbonRendererProperties]
  renderers [
   {rendererClass=&NiagaraRibbonRendererProperties}
   {rendererClass=&NiagaraRibbonRendererProperties, rendererIndex=1}
   {rendererClass=&NiagaraRibbonRendererProperties, rendererIndex=2}
  ]
 }
}
summary {
 systemName=NS_Pickaxe_CarbideWeld_Trail
 emitters [
  {emitterName=Trails_Temp, rendererClasses [&NiagaraRibbonRendererProperties]}
  {emitterName=Hexagons, rendererClasses [&NiagaraSpriteRendererProperties]}
 ]
 userVariables [
  User.bIsFrontEnd {type=NiagaraBool}
  User.bIsFrontEndPreview {type=NiagaraBool}
  User.RibbonIDOffset {type=NiagaraInt32}
  User.SwingSectionIndex {type=NiagaraInt32}
  User.Variant {type=NiagaraFloat, defaultValue=4, description=""}
 ]
}
