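# Niagara idle effect for the CarbideWeld pickaxe: a delayed startup flash and hexagon sweep that leave a persistent glow, all placed through a socket reader.
# StartFlash bursts sprites and a light while Hexagons spawn along a rising, narrowing cylinder.
# PersistantGlow fades in by system age, with a system Timeline scaling its sprites and light.
# Variant selectors tint each emitter; partition-indexed visibility tags split burst particles between two sprite renderers.
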NiagaraSystem /BRCosmetics/Weapons/FORT_Melee/Pickaxe_CarbideWeld/FX/NS_Pickaxe_Carbideweld_Idle
dependencies {
 usedDataInterfaces [
  &NiagaraDataInterfaceArrayFloat2
  &NiagaraDataInterfaceArrayInt32
  &NiagaraDataInterfaceColorCurve
  &NiagaraDataInterfaceCurve
  &NiagaraDataInterfaceVector2DCurve
  &NiagaraDataInterfaceSpriteRendererInfo
  &NiagaraDataInterfaceArrayFloat
  &NiagaraDataInterfaceSocketReader
 ]
 usedDynamicInputs [
  &Vector2DFromCurve
  &Add_Float
  &MakeCustomIntFromBool
  &SetBoolByIntComparison
 ]
 usedModules [
  &FN_VariantManager
  &self:SystemSpawnScript.SystemScriptSource.SystemScriptGraph.NiagaraNodeAssignment_0.SetVariables_#1
  &SystemState
  &self:SystemSpawnScript.SystemScriptSource.SystemScriptGraph.NiagaraNodeAssignment_1.SetVariables_#2
  &Timeline
 ]
}
emitters {
 Hexagons {
  emitterUpdateScript [
   &EmitterState(Life Cycle Mode=::NewEnumerator1, Loop Behavior=::NewEnumerator1, Loop Delay=0.5, Loop Duration=0.3, UseLoopDelay=true)
   &SpawnRate(SpawnRate=30, Use Spawn Probability=false)
  ]
  particleSpawnScript [
   &InitializeParticle(
    Lifetime Max=0.4
    Lifetime Min=0.2
    Lifetime Mode=::NewEnumerator1
    Sprite Size Mode=::NewEnumerator3
    Uniform Sprite Size=&Multiply_Float(
     A=&RandomRangeFloat(Minimum=4, Maximum=8)
     B=&FN_FrontEndFloatModifier(Gameplay=1, Front End=0.75)
    )
    UsePositionOffset=false
    Write Material Random=true
   )
   &SocketLocation(
    SamplePosOffset=true
    Sampled Position Offset=&Lerp_Vector(A=<0 0 33>, B=<0 0 111>, Alpha=$Emitter.NormalizedLoopAge)
    Socket Reader=$System.Socket Reader
    Socket Sampling Mode=::NewEnumerator3
    VelocitySpeedLimit=false
   )
   &ShapeLocation(
    Cylinder Height=40
    Cylinder Radius=&Lerp_Float(
     A=22
     B=0
     Alpha=&FloatFromCurve(
      FloatCurve=DataInterface() {propertyValues {Curve {defaultValue=1, keys [{tangentMode=RCTM_User, time=0, value=0}, {tangentMode=RCTM_User, time=1, value=1}]}}}
      CurveIndex=$Emitter.NormalizedLoopAge
     )
    )
    MaskSpawnGroup=false
    Offset=<0 0 5>
    Offset Mode=::NewEnumerator0
    Override Seed=false
    Rotation Coordinate Space=::Simulation
    Shape Primitive=::NewEnumerator1
    Surface Only=false
    UseOwnerScale=true
   )
   &CameraOffset(Camera Offset Amount=40)
   &DynamicMaterialParameters(
    Index 0 Param 2=4
    Module.Index0Float0Write=true
    Module.Index0Float1Write=true
    Module.Index0Float2Write=true
    Module.Index0Float3Write=true
   ) {module=DynamicMaterialParameters001}
   &AddVelocity(
    Module.Use Falloff Distance=false
    Origin Offset=<0 0 -44>
    Override Seed=false
    Velocity Mode=::NewEnumerator1
    Velocity Origin=$Output.SocketLocation.SampledPosition
    Velocity Speed=&RandomRangeFloat(Minimum=100, Maximum=333)
   )
  ]
  particleUpdateScript [
   &Drag(Drag=&RandomRangeFloat(Minimum=4, Maximum=7), Use Linear Drag=true, Use Rotational Drag=false)
   &SolveForcesAndVelocity(Clamp Velocity=false, Limit Acceleration=false)
   &ParticleState()
   &ScaleSpriteSize(Uniform Curve Sprite Scale=DataInterface() {propertyValues {Curve {keys [
        {arriveTangent=5.201, interpMode=RCIM_Cubic, leaveTangent=5.201, tangentMode=RCTM_User, time=0, value=1}
        {arriveTangent=-0.2045, interpMode=RCIM_Cubic, leaveTangent=-0.2045, tangentMode=RCTM_User, time=0.2, value=1.7}
        {arriveTangent=-1.983, interpMode=RCIM_Cubic, leaveTangent=-1.983, tangentMode=RCTM_User, time=1, value=0}
       ]}}})
   &ScaleColor(
    Scale Alpha=&Multiply_Float(
     A=&FloatFromCurve(FloatCurve=DataInterface() {propertyValues {Curve {keys [<at 0.75 1>,<at 1 0>]}}})
     B=0.5
    )
    Scale RGB=&MakeVectorFromLinearColorRGB(Color=&VariantColorSelector(
      Variant Amount=2
      Original Color=<rgba 1 0.5158 0 1>
      Variant 1 Color=<rgba 1 0.0526 0 1>
      Variant 2 Color=<rgba 0.2749 1 0.3525 1>
      Variant 3 Color=<rgba 0.8535 0.1618 1 1>
     ))
    ScaleA=true
    ScaleRGB=true
   ) {module=ScaleColor001}
   &SubUVAnimation(
    End Frame Range Override=7
    "Sprite Renderer "=DataInterface() {propertyValues {SpriteRenderer=&self:Hexagons_0.NiagaraSpriteRendererProperties_1}}
    UseEndFrame=true
    UseStartFrame=true
   )
  ]
  rendererClasses [&NiagaraSpriteRendererProperties]
  renderers [{rendererClass=&NiagaraSpriteRendererProperties}]
 }
 PersistantGlow {
  emitterUpdateScript [
   &EmitterState(Life Cycle Mode=::NewEnumerator1, Loop Behavior=::NewEnumerator1, Loop Delay=0.5, Loop Duration Mode=::NewEnumerator1, UseLoopDelay=true)
   &SpawnBurst_Instantaneous(Spawn Count=2, Use Spawn Probability=false, UseLoopCountLimit=false)
  ]
  particleSpawnScript [
   &PartitionParticles()
   &InitializeParticle(
    Lifetime=5
    Sprite Size=&SelectVector2DFromArray(
     Direct Array Index=$Particles.PartitionParticles.Partition
     Vector 2D Selection Array=DataInterface() {propertyValues {FloatData [<55 125>, <75 75>]}}
    )
    Sprite Size Mode=::NewEnumerator4
    UsePositionOffset=false
    Write Material Random=true
   )
   &SocketLocation(
    SamplePosOffset=true
    Sampled Position Offset=<0 0 65>
    Socket Reader=$System.Socket Reader
    Socket Sampling Mode=::NewEnumerator3
    VelocitySpeedLimit=false
   )
   /SetVariables(
    Particles.GlowAlignment:Vector3f=<0 0 1>
    Particles.GlowFlashPos:NiagaraPosition=&AddVectorToPosition(
     Position=$Particles.Position
     Vector=&TransformVector(Vector=<0 0 65>, Destination Space=::Simulation, Apply Scale=true, Transform Vector as Position=true)
    )
   ) {isSetParameters=true}
   /SetVariables(Particles.VisibilityTag=&SelectIntFromArray(Direct Array Index=$Particles.PartitionParticles.Partition, Int Selection Array=DataInterface() {propertyValues {IntData [0,1]}})) {isSetParameters=true}
   &DynamicMaterialParameters(
    Index 0 Param 1=0.7
    Index 0 Param 2=0.1
    Index 0 Param 3=0.1
    Module.Index0Float0Write=true
    Module.Index0Float1Write=true
    Module.Index0Float2Write=true
    Module.Index0Float3Write=true
   )
   &CameraOffset(Camera Offset Amount=&SelectFloatFromArray(Direct Array Index=$Engine.QualityLevel, Float Selection Array=DataInterface() {propertyValues {FloatData [-10,-10,0,0,0]}}))
  ]
  particleUpdateScript [
   &ParticleState(Kill Particles When Lifetime Has Elapsed=false, Let Infinitely Lived Particles Die When Emitter Deactivates=true)
   &ScaleSpriteSize(
    Scale Factor=&Multiply_Vector2D_ByFloat(Float=&Lerp_Float(A=1, B=2.5, Alpha=$System.Timeline.Float), Vector2D=<1 1>)
    Scale Sprite Size Mode=::NewEnumerator2
   )
   &ScaleColor(
    Scale Alpha=&FloatFromCurve(
     FloatCurve=DataInterface() {propertyValues {Curve {defaultValue=1, keys [<at 0.6 0 RCIM_Cubic>,<at 0.8 1 RCIM_Cubic>]}}}
     CurveIndex=$Engine.System.Age
    )
    Scale RGB=&MakeVectorFromLinearColorRGB(Color=&VariantColorSelector(
      Variant Amount=2
      Original Color=<rgba 2 0.7724 0.0264 1>
      Variant 1 Color=<rgba 2 0.1833 0 1>
      Variant 2 Color=<rgba 0.8284 1.5 0.5612 1>
      Variant 3 Color=<rgba 1.5736 0.2582 2 1>
     ))
    ScaleA=true
    ScaleRGB=true
   )
   &FN_Light_Attributes(
    Color=$Particles.Color
    Create Color Attribute=true
    Create Exponent Attribute=true
    Create Position Attribute=true
    Create Radius Attribute=true
    Exponent=2
    Position=&AddVectorToPosition(
     Position=$Particles.Position
     Vector=&TransformVector(Vector=<0 0 25>, Destination Space=::Simulation, Apply Scale=true, Transform Vector as Position=true)
    )
    Radius=&Lerp_Float(A=100, B=150, Alpha=$System.Timeline.Float)
    Scale Color=&FloatFromCurve(
     FloatCurve=DataInterface() {propertyValues {Curve {defaultValue=1, keys [{tangentMode=RCTM_User, time=0, value=0}, {tangentMode=RCTM_User, time=0.5, value=1}]}}}
     CurveIndex=$Engine.System.Age
     Scale Curve=&Lerp_Float(A=1, B=2, Alpha=$System.Timeline.Float)
    )
   )
  ]
  rendererClasses [&NiagaraSpriteRendererProperties, &NiagaraLightRendererProperties]
  renderers [
   {rendererClass=&NiagaraSpriteRendererProperties}
   {rendererClass=&NiagaraSpriteRendererProperties, rendererIndex=1}
   {rendererClass=&NiagaraLightRendererProperties, rendererIndex=2}
  ]
 }
 StartFlash {
  emitterUpdateScript [
   &EmitterState(Life Cycle Mode=::NewEnumerator1, Loop Behavior=::NewEnumerator1, Loop Delay=0.5, Loop Duration=1, UseLoopDelay=true)
   &SpawnBurst_Instantaneous(Spawn Count=2, Use Spawn Probability=false, UseLoopCountLimit=false)
  ]
  particleSpawnScript [
   &PartitionParticles()
   &InitializeParticle(
    Color Mode=::NewEnumerator0
    Lifetime=0.3
    Sprite Size=&SelectVector2DFromArray(
     Direct Array Index=$Particles.PartitionParticles.Partition
     Vector 2D Selection Array=DataInterface() {propertyValues {FloatData [<150 150>, <35 75>]}}
    )
    Sprite Size Mode=::NewEnumerator4
    UsePositionOffset=false
    Write Material Random=true
   )
   &SocketLocation(SamplePosOffset=true, Socket Reader=$System.Socket Reader, Socket Sampling Mode=::NewEnumerator3, VelocitySpeedLimit=false)
   &CameraOffset(Camera Offset Amount=50)
   /SetVariables(Particles.VisibilityTag=&SelectIntFromArray(Direct Array Index=$Particles.PartitionParticles.Partition, Int Selection Array=DataInterface() {propertyValues {IntData [0,1]}})) {isSetParameters=true}
   /SetVariables(
    Particles.GlowAlignment:Vector3f=<0 0 1>
    Particles.GlowFlashPos:NiagaraPosition=&AddVectorToPosition(
     Position=$Particles.Position
     Vector=&TransformVector(Vector=<0 0 65>, Destination Space=::Simulation, Apply Scale=true, Transform Vector as Position=true)
    )
   ) {isSetParameters=true}
   &DynamicMaterialParameters(
    Index 0 Param 1=1
    Module.Index0Float0Write=true
    Module.Index0Float1Write=true
    Module.Index0Float2Write=true
    Module.Index0Float3Write=true
   )
  ]
  particleUpdateScript [
   &ParticleState()
   &ScaleColor(
    Scale Alpha=&FloatFromCurve(FloatCurve=DataInterface() {propertyValues {Curve {
        defaultValue=1
        keys [<at 0 1 RCIM_Cubic>, {arriveTangent=0.0208, interpMode=RCIM_Cubic, leaveTangent=0.0208, tangentMode=RCTM_User, time=0.75, value=0}]
       }}})
    Scale RGB=&MakeVectorFromLinearColorRGB(Color=&VariantColorSelector(
      Variant Amount=2
      Original Color=&ColorFromCurve(ColorCurve=DataInterface() {propertyValues {
         AlphaCurve {keys [<at 0 1 RCIM_Cubic>,<at 1 0 RCIM_Cubic>]}
         BlueCurve {keys [<at 0 0 RCIM_Cubic>,<at 0.6997 0 RCIM_Cubic>]}
         GreenCurve {keys [<at 0 0.4032 RCIM_Cubic>,<at 0.6997 0.1195 RCIM_Cubic>]}
         RedCurve {keys [<at 0 0.75 RCIM_Cubic>,<at 0.6997 0.4479 RCIM_Cubic>]}
        }})
      Variant 1 Color=&ColorFromCurve(ColorCurve=DataInterface() {propertyValues {
         AlphaCurve {keys [<at 0 1 RCIM_Cubic>,<at 1 0 RCIM_Cubic>]}
         BlueCurve {keys [<at 0 0 RCIM_Cubic>,<at 0.6997 0.0027 RCIM_Cubic>]}
         GreenCurve {keys [<at 0 0.0453 RCIM_Cubic>,<at 0.6997 0 RCIM_Cubic>]}
         RedCurve {keys [<at 0 0.75 RCIM_Cubic>,<at 0.6997 0.3333 RCIM_Cubic>]}
        }})
      Variant 2 Color=&ColorFromCurve(ColorCurve=DataInterface() {propertyValues {
         AlphaCurve {keys [<at 0 1 RCIM_Cubic>,<at 1 0 RCIM_Cubic>]}
         BlueCurve {keys [<at 0 0.1496 RCIM_Cubic>,<at 0.6997 0.0355 RCIM_Cubic>]}
         GreenCurve {keys [<at 0 0.5 RCIM_Cubic>,<at 0.6997 0.1 RCIM_Cubic>]}
         RedCurve {keys [<at 0 0.1172 RCIM_Cubic>,<at 0.6997 0.0252 RCIM_Cubic>]}
        }})
      Variant 3 Color=&ColorFromCurve(ColorCurve=DataInterface() {propertyValues {
         AlphaCurve {keys [<at 0 1 RCIM_Cubic>,<at 1 0 RCIM_Cubic>]}
         BlueCurve {keys [<at 0 0.5 RCIM_Cubic>,<at 0.6997 0.1 RCIM_Cubic>]}
         GreenCurve {keys [<at 0 0.1397 RCIM_Cubic>,<at 0.6997 0.0182 RCIM_Cubic>]}
         RedCurve {keys [<at 0 0.4713 RCIM_Cubic>,<at 0.6997 0.0875 RCIM_Cubic>]}
        }})
     ))
    ScaleA=true
    ScaleRGB=true
   )
   /SetVariables(Particles.GlowColor=&MakeLinearColorFromVectorAndFloat(
     "Vector (RGB)"=&MakeVectorFromLinearColorRGB(Color=$Particles.Color)
     "Float (Alpha)"=&FloatFromCurve(
      FloatCurve=DataInterface() {propertyValues {Curve {
         defaultValue=1
         keys [
          {arriveTangent=9.0725, interpMode=RCIM_Cubic, leaveTangent=9.0725, tangentMode=RCTM_User, time=0, value=0}
          {arriveTangent=0.0936, interpMode=RCIM_Cubic, leaveTangent=0.0936, tangentMode=RCTM_User, time=0.1408, value=0.9847}
          {arriveTangent=-0.0405, interpMode=RCIM_Cubic, leaveTangent=-0.0405, tangentMode=RCTM_User, time=1, value=0}
         ]
        }}}
      Scale Curve=0.5
     )
    )) {isSetParameters=true}
   /SetVariables(Particles.TravelGlowPos:NiagaraPosition=&AddVectorToPosition(
     Position=$Particles.Position
     Vector=&TransformVector(
      Vector=&Lerp_Vector(
       A=<0 0 33>
       B=<0 0 111>
       Alpha=&FloatFromCurve(FloatCurve=DataInterface() {propertyValues {Curve {defaultValue=1, keys [{tangentMode=RCTM_User, time=0, value=0}, {tangentMode=RCTM_User, time=0.75, value=1}]}}})
      )
      Destination Space=::Simulation
      Apply Scale=true
      Transform Vector as Position=true
     )
    )) {isSetParameters=true}
   &FN_Light_Attributes(
    Color=$Particles.Color
    Create Color Attribute=true
    Create Exponent Attribute=true
    Create Position Attribute=true
    Create Radius Attribute=true
    Exponent=2
    Position=&AddVectorToPosition(
     Position=$Particles.Position
     Vector=&TransformVector(Vector=<0 0 25>, Destination Space=::Simulation, Apply Scale=true, Transform Vector as Position=true)
    )
    Radius=175
    Scale Color=&FloatFromCurve(
     FloatCurve=DataInterface() {propertyValues {Curve {
        defaultValue=0
        keys [
         {interpMode=RCIM_Cubic, tangentMode=RCTM_User, time=0, value=0}
         {interpMode=RCIM_Cubic, tangentMode=RCTM_User, time=0.5, value=1}
         {tangentMode=RCTM_User, time=1, value=0}
        ]
       }}}
     Scale Curve=10
    )
   )
  ]
  rendererClasses [&NiagaraSpriteRendererProperties, &NiagaraLightRendererProperties]
  renderers [
   {rendererClass=&NiagaraSpriteRendererProperties}
   {rendererClass=&NiagaraSpriteRendererProperties, rendererIndex=1}
   {rendererClass=&NiagaraLightRendererProperties, rendererIndex=2}
  ]
 }
}
summary {
 systemName=NS_Pickaxe_Carbideweld_Idle
 emitters [
  {emitterName=StartFlash, rendererClasses [&NiagaraSpriteRendererProperties, &NiagaraLightRendererProperties]}
  {emitterName=Hexagons, rendererClasses [&NiagaraSpriteRendererProperties]}
  {
   emitterName=PersistantGlow
   rendererClasses [&NiagaraSpriteRendererProperties, &NiagaraLightRendererProperties]
  }
 ]
 userVariables [
  User.bIsFrontEnd {type=NiagaraBool}
  User.bIsFrontEndPreview {type=NiagaraBool}
  User.bisSwinging {type=NiagaraBool}
  User.SwingSectionIndex {type=NiagaraInt32}
  User.Variant {type=NiagaraFloat, defaultValue=0, description=""}
 ]
}
